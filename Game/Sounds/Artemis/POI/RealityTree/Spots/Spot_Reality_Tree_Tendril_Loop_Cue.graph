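# Looping ambient spot cue for the Reality Tree tendrils, randomly picking one of two looping waves.
# The random pick feeds close 2D and close attenuation branches, mixed with the 2D branch louder.
# Two nested endlessly looping envelopers add slow pitch drift and a gentle volume swell, behind a short random start delay.
# The quality switch fills only its first two slots, leaving the other two empty.
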SoundCue /Game/Sounds/Artemis/POI/RealityTree/Spots/Spot_Reality_Tree_Tendril_Loop_Cue
root=QualityLevel_0
nodes [
 {
  id=Attenuation_6
  children [Random_0]
  properties {attenuationSettings=&/Spot_Reality_Tree_Tendril_Close_Att}
 }
 {
  id=Attenuation_9
  children [Random_0]
  properties {attenuationSettings=&/Spot_Reality_Tree_Tendril_Close_2D_Att}
 }
 {id=Delay_0, children [Enveloper_2], properties {delayMax=0.75, delayMin=0.2}}
 {
  id=Enveloper_1
  children [Modulator_0]
  properties {
   bLoop=true
   bLoopIndefinitely=true
   loopEnd=60
   pitchCurve {editorCurveData {keys [
      {interpMode=RCIM_Cubic, time=0}
      <at 15 0.95 RCIM_Cubic>
      {arriveTangent=0.0072, interpMode=RCIM_Cubic, leaveTangent=0.0072, tangentMode=RCTM_User, time=30}
      <at 45 1.1 RCIM_Cubic>
      {time=60}
     ]}}
   volumeCurve {editorCurveData {keys [<at 0 1>]}}
  }
 }
 {
  id=Enveloper_2
  children [Enveloper_1]
  properties {
   bLoop=true
   bLoopIndefinitely=true
   loopEnd=20
   pitchCurve {editorCurveData {keys [{time=0}]}}
   volumeCurve {editorCurveData {keys [
      {arriveTangent=-0.01, interpMode=RCIM_Cubic, time=0, value=1}
      <at 10 0.8 RCIM_Cubic>
      {interpMode=RCIM_Cubic, leaveTangent=0.01, time=20, value=1}
     ]}}
  }
 }
 {id=Mixer_2, children [Attenuation_9, Attenuation_6], properties {inputVolume [2,1]}}
 {id=Modulator_0, children [Mixer_2], properties {pitchMax=1, pitchMin=1}}
 {id=QualityLevel_0, children [Delay_0,Delay_0,null,null]}
 {id=Random_0, children [WavePlayer_5,WavePlayer_6], properties {weights [1,1]}}
 {
  id=WavePlayer_5
  properties {bLooping=true, soundWaveAssetPtr=&/AMB_Biome_RealityForest_Tendril_Loop_01}
 }
 {
  id=WavePlayer_6
  properties {bLooping=true, soundWaveAssetPtr=&/AMB_Biome_RealityForest_Tendril_Loop_02}
 }
]
properties {
 maxDistance=3050
 soundClassObject=/Game/Sounds/FortSoundClassesAndModes/Fort_Water_Ambient
 volumeMultiplier=1
 attenuationOverrides {
  bAttenuateWithLPF=true
  falloffDistance=6000
  lPFFrequencyAtMax=500
  nonSpatializedRadiusStart=700
  stereoSpread=1000
  attenuationShapeExtents {x=925}
 }
 concurrencyOverrides {maxCount=1, resolutionRule=PreventNew}
}
